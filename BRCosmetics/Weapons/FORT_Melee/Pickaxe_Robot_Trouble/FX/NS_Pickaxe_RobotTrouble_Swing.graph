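# Niagara swing trail effect for the Robot Trouble pickaxe, drawn as short-lived ribbons.
# One MainTrail emitter spawns four anim trails, each initialized from bone or socket index pairs, and feeds two ribbon renderers.
# Merge-helper switches keyed on spawn group give each trail its color and alpha fade curve, combining gray, teal and red streaks in a single emitter.
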NiagaraSystem /BRCosmetics/Weapons/FORT_Melee/Pickaxe_Robot_Trouble/FX/NS_Pickaxe_RobotTrouble_Swing
dependencies {
 usedDataInterfaces [&NiagaraDataInterfaceCurve, &NiagaraDataInterfaceSocketReader]
 usedDynamicInputs []
 usedModules [
  &self:SystemSpawnScript.SystemScriptSource.SystemScriptGraph.NiagaraNodeAssignment_1.SetVariables_#1
  &SystemState
 ]
}
emitters {MainTrail {
  emitterSpawnScript [&FN_RibbonIDOffsetUtility()]
  emitterUpdateScript [
   &EmitterState()
   &FN_SpawnAnimTrail(Number of Anim Trails=4)
  ]
  particleSpawnScript [
   &FN_InitializeAnimTrailBoneIndecies()
   &FN_InitializeAnimTrailBoneIndecies(AnimTrail Selection=1, First Bone or Socket Index=3, Second Bone or Socket Index=2) {module=FN_InitializeAnimTrailBoneIndecies001}
   &FN_InitializeAnimTrailBoneIndecies(AnimTrail Selection=2, First Bone or Socket Index=4, Second Bone or Socket Index=3) {module=FN_InitializeAnimTrailBoneIndecies004}
   &FN_InitializeAnimTrailBoneIndecies(AnimTrail Selection=3, First Bone or Socket Index=7, Second Bone or Socket Index=6) {module=FN_InitializeAnimTrailBoneIndecies003}
   &FN_InitializeAnimTrail(
    Color=&MergeHelper_ColorSwitch(
     0=<rgba 0.2556 0.2565 0.2747 0.7>
     1=<rgba 0.0309 0.3095 0.278 0.4>
     2=<rgba 0.0309 0.3095 0.278 0.4>
     3=<rgba 0.8 0.12 0.1245 0.3>
     Index=&SpawnGroup()
     Num=::NewEnumerator2
    )
    Lifetime=0.22
   )
   &AddVelocity(Velocity=<0 400 0>) {enabled=false}
  ]
  particleUpdateScript [
   &SolveForcesAndVelocity(Clamp Velocity=false, Limit Acceleration=false)
   &FN_UpdateAnimTrail() {module=FN_UpdateAnimTrail001}
   &ScaleColor(
    Scale Alpha=&MergeHelper_FloatSwitch(
     0=&FloatFromCurve(FloatCurve=DataInterface() {propertyValues {Curve {keys [{arriveTangent=-18.1546, interpMode=RCIM_Cubic, leaveTangent=-18.1546, tangentMode=RCTM_User, time=0.3, value=1}, <at 0.5 0 RCIM_Cubic>]}}})
     1=&FloatFromCurve(FloatCurve=DataInterface() {propertyValues {Curve {keys [{arriveTangent=-11.0837, interpMode=RCIM_Cubic, leaveTangent=-11.0837, tangentMode=RCTM_User, time=0.2, value=1}, <at 0.4 0 RCIM_Cubic>]}}})
     2=&FloatFromCurve(FloatCurve=DataInterface() {propertyValues {Curve {keys [{arriveTangent=-11.0837, interpMode=RCIM_Cubic, leaveTangent=-11.0837, tangentMode=RCTM_User, time=0.2, value=1}, <at 0.4 0 RCIM_Cubic>]}}})
     3=&FloatFromCurve(FloatCurve=DataInterface() {propertyValues {Curve {keys [{arriveTangent=-11.0837, interpMode=RCIM_Cubic, leaveTangent=-11.0837, tangentMode=RCTM_User, time=0.2, value=1}, <at 0.4 0 RCIM_Cubic>]}}})
     Index=&SpawnGroup()
     Num=::NewEnumerator2
    )
    ScaleA=true
    ScaleRGB=true
   )
  ]
  rendererClasses [&NiagaraRibbonRendererProperties]
  renderers [
   {rendererClass=&NiagaraRibbonRendererProperties}
   {rendererClass=&NiagaraRibbonRendererProperties, rendererIndex=1}
  ]
 }}
summary {
 systemName=NS_Pickaxe_RobotTrouble_Swing
 emitters [{emitterName=MainTrail, rendererClasses [&NiagaraRibbonRendererProperties]}]
 userVariables [User.RibbonIDOffset {type=NiagaraInt32}]
}
