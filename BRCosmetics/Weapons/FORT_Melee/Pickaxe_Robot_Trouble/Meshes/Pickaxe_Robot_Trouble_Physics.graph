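# Physics asset for the Robot Trouble pickaxe, a two-handed melee cosmetic weapon mesh.
# Three capsule bodies approximate the handle, a long iron plate and a circular iron plate.
# Two constraints join each plate directly to the handle, both with twist locked.
# The circular plate also locks one swing axis, and small swing angles are set on both joints.
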PhysicsAsset /BRCosmetics/Weapons/FORT_Melee/Pickaxe_Robot_Trouble/Meshes/Pickaxe_Robot_Trouble_Physics
bodies {
 CircleIronPlate [{shapeType=Capsule, length=0.0595, radius=17.6548, center=<-0.6956 -1.7926 -1.1067>, rotation=<rot 4.4411 -138.8725 -90>}]
 LongIronPlate [{shapeType=Capsule, length=64.0928, radius=8.5005, center=<2.8399 0.5209 0.1312>, rotation=<rot 0.0196 -88.9186 -90.5211>}]
 Melee_TwoHanded_Handle [{shapeType=Capsule, length=108.3097, radius=2.2775, center=<1.4202 -1.0431 26.5785>, rotation=<rot -0.4623 -90.0007 -0.1887>}]
}
constraints [
 {
  bone1Name=LongIronPlate
  bone2Name=Melee_TwoHanded_Handle
  swing1LimitDegrees=1.5
  swing2LimitDegrees=0.1
  twistLimitDegrees=45
  twistMotion=Locked
 }
 {
  bone1Name=CircleIronPlate
  bone2Name=Melee_TwoHanded_Handle
  swing1LimitDegrees=45
  swing1Motion=Locked
  swing2LimitDegrees=2
  twistLimitDegrees=45
  twistMotion=Locked
 }
]
